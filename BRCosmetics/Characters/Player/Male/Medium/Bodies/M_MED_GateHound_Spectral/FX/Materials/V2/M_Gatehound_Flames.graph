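# Translucent, unlit, two-sided Niagara mesh-particle material for wispy spectral flames on the Gatehound character cosmetic.
# A panning noise distorts a second noise sample, which combines with UV gradients to drive dark-to-green emissive color and opacity.
# Opacity also fades along V, by view angle and by quality-switched depth fade.
# World position offset adds sine waves and a dynamic-parameter moving-average pull, with optional SplineThicken; vertex color green offsets time.
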Material /BRCosmetics/Characters/Player/Male/Medium/Bodies/M_MED_GateHound_Spectral/FX/Materials/V2/M_Gatehound_Flames
graph {
 properties {
  bUsedWithNiagaraMeshParticles=true
  blendMode=BLEND_Translucent
  pixelDepthOffsetMode=PDOM_Legacy
  shadingModel=MSM_Unlit
  twoSided=true
  usedShadingModels=Unlit
 }
 nodes [
  n1 = ScalarParameter() {defaultValue=10, parameterName=Thickness}
  Add(A=StaticSwitchParameter(
    True=MaterialFunctionCall(WidthBase=@n1, WidthTip=@n1) {materialFunction=&SplineThicken}.WorldPositionOffset
    False=Constant()
   ) {parameterName=UseSplineThicken})
  Constant()
  n2 = Constant3Vector()
  n3 = Constant3Vector() {constant=<rgba 0.02 0.02 0.02 0>}
  n4 = MaterialFunctionCall(Float2=TextureCoordinate()) {materialFunction=&BreakOutFloat2Components}
  n5 = StaticSwitchParameter(True=StaticSwitchParameter(True=@n4.R, False=OneMinus(None=@n4.R)) {parameterName=FlipHorizontalGradient}, False=@n4.G) {parameterName=HorizontalPanning}
  n6 = MaterialFunctionCall(Float2=TextureCoordinate()) {materialFunction=&BreakOutFloat2Components}
  n7 = SphereMask(A=StaticSwitchParameter(True=@n6.G, False=@n6.G) {parameterName=HorizontalPanning}, B=Constant() {r=0.5}) {attenuationRadius=0.5, hardnessPercent=0}
  n8 = SmoothStep(Value=StaticSwitchParameter(True=@n5, False=Multiply(A=@n5, B=@n7)) {parameterName=IsTiling})
  n9 = TextureCoordinate()
  n10 = MaterialFunctionCall(Float2=@n9) {materialFunction=&BreakOutFloat2Components}
  n11 = MaterialFunctionCall(X=@n10.R, Y=OneMinus(None=@n10.G)) {materialFunction=&MakeFloat2}
  n12 = TextureCoordinate() {vTiling=0.5}
  n13 = MaterialFunctionCall(Float2=@n12) {materialFunction=&BreakOutFloat2Components}
  n14 = StaticSwitchParameter(
   True=MaterialFunctionCall(X=@n13.R, Y=@n13.R) {materialFunction=&MakeFloat2}.Result
   False=@n12
  ) {parameterName=HorizontalPanning}
  n15 = NamedRerouteUsage()
  n16 = VectorParameter() {
   parameterName=DistortTextureControl
   channelNames {a=SizeY_Dist, b=SizeX_Dist, g=SpeedY_Dist, r=SpeedX_Dist}
   defaultValue=<rgba 0.1 0.5 2 0.75>
  }
  n17 = AppendVector(A=@n16.SpeedX_Dist, B=@n16.SpeedX_Dist)
  n18 = AppendVector(A=@n16.SizeY_Dist, B=@n16.SizeY_Dist)
  n19 = AppendVector(A=@n16.SizeX_Dist, B=@n16.SizeX_Dist)
  n20 = TextureSampleParameter2D(UVs=Multiply(
    A=Panner(Coordinate=@n14, Time=@n15, Speed=@n17) {speedY=0.25}
    B=StaticSwitchParameter(True=@n18, False=@n19) {parameterName=HorizontalPanning}
   )) {
   group=Wispy Flame
   parameterName=Flame Texture
   texture=/BRCosmetics/Athena/Effects/Particles/Skydiving/NS_Contrail_Chainmail/T_WispyFireNoise03
  }
  n21 = VectorParameter() {parameterName=MainTextureControl, channelNames {a=SizeY, b=SizeX, g=SpeedY, r=SpeedX}, defaultValue=<rgba 0 1.25 1 1>}
  n22 = Panner(
   Coordinate=Add(A=StaticSwitchParameter(True=@n11.Result, False=@n9) {parameterName=HorizontalPanning}, B=Multiply(A=@n20.R) {constB=0.1})
   Time=NamedRerouteUsage()
   Speed=AppendVector(A=@n21.SpeedX, B=@n21.SpeedX)
  ) {speedX=1}
  n23 = AppendVector(A=@n21.SizeX, B=@n21.SizeX)
  n24 = TextureSampleParameter2D(UVs=MaterialFunctionCall(XY=Multiply(A=@n22, B=@n23)) {materialFunction=&Swizzle}.YX) {
   group=Wispy Flame
   parameterName=Flame Texture
   texture=&T_WispyFireNoise03
  }
  n25 = Multiply(A=@n8, B=@n24.R)
  n26 = Constant3Vector() {constant=<rgba 0.1637 1 0.4503 1>}
  n27 = Constant3Vector() {constant=<rgba 0.1394 2 0.2649 1>}
  n28 = VectorParameter() {parameterName=ColorControl, channelNames {a=AlphaMax, b=AlphaMin, g=ColorGradientMax, r=ColorGradientMin}, defaultValue=<rgba 1 1 0.02 0.1>}
  n29 = LinearInterpolate(A=@n25, B=@n8, Alpha=@n8)
  n30 = LinearInterpolate(A=@n25, B=@n29, Alpha=Constant() {r=0.75})
  MaterialFunctionCall(Emissive=LinearInterpolate(
    A=LinearInterpolate(A=@n2, B=@n3, Alpha=@n25)
    B=LinearInterpolate(A=@n26, B=@n27, Alpha=@n25)
    Alpha=SmoothStep(Min=@n28.ColorGradientMin, Max=@n28.ColorGradientMin, Value=@n30)
   )) {materialFunction=&MF_FX_EmissiveAdjustments}
  n31 = DynamicParameter() {parameterIndex=1, paramNames [Moving Average X, Moving Average Y, Moving Average Z, WPO Intensity]}
  n32 = ComponentMask(None=TextureCoordinate()) {g=true}
  n33 = OneMinus(None=@n32)
  n34 = LinearInterpolate(A=ScalarParameter() {defaultValue=2, parameterName=WPO Frequency}, B=Constant(), Alpha=@n32)
  n35 = NamedRerouteUsage()
  n36 = ScalarParameter() {defaultValue=2, parameterName=WPO Speed}
  n37 = Subtract(
   A=SmoothStep(Value=Multiply(A=@n34, B=@n33)) {constMin=0.25}
   B=Add(
    A=DynamicParameter() {defaultValue=<rgba 0 0.5 0 1>, paramNames [Time Offset, Param2, Param3, Param4]}.Time Offset
    B=Multiply(A=@n35, B=@n36)
   )
  )
  n38 = AppendVector(A=Constant(), B=Cosine(None=@n37) {desc=Y})
  n39 = Sine(None=@n37) {desc=Z, period=3}
  n40 = Multiply(A=@n33, B=Multiply(A=AppendVector(A=@n38, B=@n39), B=ScalarParameter() {defaultValue=0.2, parameterName=WPO Amplitude}) {constB=3})
  n41 = Multiply(A=@n33, B=@n33)
  n42 = Subtract(
   A=MaterialFunctionCall(R=@n31.Moving Average X, G=@n31.Moving Average X, B=@n31.Moving Average X) {materialFunction=&AppendMany}.RGB
   B=ParticlePositionWS()
  )
  Multiply(A=@n31.WPO Intensity, B=Add(A=Transform(None=@n40) {transformSourceType=TRANSFORMSOURCE_Instance}, B=Multiply(A=@n41, B=@n42)))
  n43 = SmoothStep(Min=@n28.AlphaMin, Max=@n28.AlphaMin, Value=@n29) {constMax=0.5}
  n44 = TextureCoordinate()
  n45 = OneMinus(None=SmoothStep(Min=ScalarParameter() {defaultValue=0.5, parameterName=Param_1}, Value=ComponentMask(None=@n44) {g=true}) {constMin=0.9})
  n46 = ScalarParameter() {defaultValue=1}
  n47 = DepthFade() {fadeDistanceDefault=5}
  n48 = Constant() {r=1}
  Multiply(
   A=Multiply(
    A=Multiply(A=@n43, B=@n45)
    B=MaterialFunctionCall("FadeSensitivity (optional)"=@n46) {materialFunction=&FadeBasedOnViewAngle}.Result
   )
   B=QualitySwitch(Default=@n47, Low=@n48, High=@n47, Medium=@n48, Epic=@n47)
  )
  n49 = Add(A=Time(), B=VertexColor().G)
  NamedRerouteDeclaration(None=@n49) {name=RandomTime, nodeColor=<rgba 0 0.5312 1 1>}
  Reroute(None=@n49)
 ]
}
